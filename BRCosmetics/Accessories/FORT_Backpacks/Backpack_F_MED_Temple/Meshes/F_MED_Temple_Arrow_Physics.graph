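# Physics asset for the F_MED_Temple backpack cosmetic, giving its eight quiver arrows and a satchel bodies for physical motion.
# Each arrow has a small capsule, the satchel a much larger one, and all are constrained directly to a capsule on the pack bone in a star layout.
# Tight swing and twist limits allow the arrows only slight movement, while the satchel gets a much wider range to swing.
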PhysicsAsset /BRCosmetics/Accessories/FORT_Backpacks/Backpack_F_MED_Temple/Meshes/F_MED_Temple_Arrow_Physics
bodies {
 arrow1 [{shapeType=Capsule, length=2, radius=2, center=<2.0245 0.6701 -0.6813>, rotation=<rot -60.5565 -179.9999 -164.5736>}]
 arrow2 [{shapeType=Capsule, length=2, radius=2, center=<2.0241 0.6701 -0.681>, rotation=<rot -60.5255 -179.9999 -164.5757>}]
 arrow3 [{shapeType=Capsule, length=2, radius=2, center=<2.0241 0.67 -0.681>, rotation=<rot -60.5382 -179.9998 -164.5582>}]
 arrow4 [{shapeType=Capsule, length=2, radius=2, center=<2.0242 0.6701 -0.681>, rotation=<rot -60.5349 -179.9999 -164.5725>}]
 arrow5 [{shapeType=Capsule, length=2, radius=2, center=<2.0243 0.6701 -0.6811>, rotation=<rot -60.5463 -179.9999 -164.5687>}]
 arrow6 [{shapeType=Capsule, length=2, radius=2, center=<2.0242 0.6701 -0.6811>, rotation=<rot -60.5354 -179.9999 -164.5748>}]
 arrow7 [{shapeType=Capsule, length=2, radius=2, center=<2.0243 0.6701 -0.6811>, rotation=<rot -60.5418 -179.9999 -164.5699>}]
 arrow8 [{shapeType=Capsule, length=2, radius=2, center=<2.5245 0.7725 -1.2214>, rotation=<rot -60.5334 -179.9999 -164.5706>}]
 dyn_satchel_Main [{shapeType=Capsule, length=10.1978, radius=10.0681, center=<15.3754 0.1625 -0.3374>, rotation=<rot 6.3915 12.839 -90>}]
 pack [{shapeType=Capsule, length=2, radius=2}]
}
constraints [
 {bone1Name=arrow1, bone2Name=pack, swing1LimitDegrees=5, swing2LimitDegrees=5, twistLimitDegrees=10}
 {bone1Name=arrow2, bone2Name=pack, swing1LimitDegrees=5, swing2LimitDegrees=5, twistLimitDegrees=10}
 {bone1Name=arrow3, bone2Name=pack, swing1LimitDegrees=5, swing2LimitDegrees=5, twistLimitDegrees=10}
 {bone1Name=arrow4, bone2Name=pack, swing1LimitDegrees=5, swing2LimitDegrees=5, twistLimitDegrees=10}
 {bone1Name=arrow5, bone2Name=pack, swing1LimitDegrees=10, swing2LimitDegrees=5, twistLimitDegrees=10}
 {bone1Name=arrow6, bone2Name=pack, swing1LimitDegrees=5, swing2LimitDegrees=5, twistLimitDegrees=10}
 {bone1Name=arrow7, bone2Name=pack, swing1LimitDegrees=7, swing2LimitDegrees=5, twistLimitDegrees=10}
 {bone1Name=arrow8, bone2Name=pack, swing1LimitDegrees=5, swing2LimitDegrees=5, twistLimitDegrees=10}
 {bone1Name=dyn_satchel_Main, bone2Name=pack, swing1LimitDegrees=45, swing2LimitDegrees=45, twistLimitDegrees=45}
]
