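# Physics asset for the F_MED Temple backpack cosmetic, defining capsule bodies and constraints for its dangling bag, ring, loops, keychain and ropes.
# Tiny MSTR bodies link the ring, keychain and rope chains to the pack, which is constrained to a two-capsule root.
# The bag is a two-link chain with loops branching off its first link.
# Tight limits make the ring hinge-like and keep the three-link bottom rope from twisting, while the keychain swings freely.
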PhysicsAsset /BRCosmetics/Accessories/FORT_Backpacks/Backpack_F_MED_Temple/Meshes/F_MED_Temple_Pack_Physics
bodies {
 dyn_bagRing [{shapeType=Capsule, length=2.0944, radius=0.5401, center=<0.0736 -0.1225 -0.8659>, rotation=<rot 2.4052 -10.036 -79.8231>}]
 dyn_bagRing_MSTR [{shapeType=Capsule, length=0.505, radius=0.505}]
 dyn_bag_01 [{shapeType=Capsule, length=1.857, radius=1, center=<1.9954 0 0>, rotation=<rot -90 0 -0.0001>}]
 dyn_bag_02 [{shapeType=Capsule, length=5.7607, radius=1.5902, center=<4.7712 -0.498 0.2151>, rotation=<rot -0.568 83.5158 -101.3851>}]
 dyn_keychain_01 [{shapeType=Capsule, length=5.5063, radius=0.8, shapeName=dyn_keychain_01_capsule, center=<3.1555 0 0>, rotation=<rot -90 180 180>}]
 dyn_keychain_MSTR [{shapeType=Capsule, length=0.505, radius=0.505}]
 dyn_loop_01 [{shapeType=Capsule, length=1.4713, radius=0.5, center=<1.1433 0 0>, rotation=<rot -90 0 0>}]
 dyn_loop_02 [{shapeType=Capsule, length=1.808, radius=0.5, center=<1.841 0 0>, rotation=<rot -90 0 0>}]
 dyn_rope_BTM_01 [{shapeType=Capsule, length=2.0838, radius=1.0966, center=<2.6699 0 0.0001>, rotation=<rot -90 0 0>}]
 dyn_rope_BTM_02 [{shapeType=Capsule, length=1.4232, radius=1.2953, center=<2.9781 0 0.0001>, rotation=<rot -90 0 0>}]
 dyn_rope_BTM_03 [{shapeType=Capsule, length=0.2437, radius=1.9493, center=<1.8484 0 0>, rotation=<rot -90 0 0>}]
 dyn_rope_BTM_MSTR [{shapeType=Capsule, length=0.51, radius=0.5}]
 dyn_rope_top_01 [{shapeType=Capsule, length=0.7323, radius=1.7616, center=<-0.819 0.239 0.0301>, rotation=<rot 0 80.404 -81.3801>}]
 dyn_rope_top_MSTR [{shapeType=Capsule, length=0.505, radius=0.505}]
 pack [{shapeType=Capsule, length=10.3559, radius=2.4361, center=<-5.9845 3.6608 -16.3602>, rotation=<rot -8.8781 -90.0937 -1.0276>}]
 root [
  {shapeType=Capsule, length=4.5166, radius=2.4921, center=<-4.8671 -3.4946 -2.1044>, rotation=<rot 18.461 -91.1184 -12.4883>}
  {
   shapeType=Capsule
   length=13.0661
   radius=3.6338
   shapeName=root_capsule
   center=<-1.1187 -5.5452 -14.687>
   rotation=<rot 22.1605 -2.2501 26.3263>
  }
 ]
}
constraints [
 {
  bone1Name=pack
  bone2Name=root
  swing1LimitDegrees=45
  swing1Motion=Free
  swing2LimitDegrees=45
  swing2Motion=Free
  twistLimitDegrees=45
  twistMotion=Free
 }
 {
  bone1Name=dyn_bagRing_MSTR
  bone2Name=pack
  swing1LimitDegrees=45
  swing1Motion=Free
  swing2LimitDegrees=45
  swing2Motion=Free
  twistLimitDegrees=45
  twistMotion=Free
 }
 {bone1Name=dyn_rope_top_MSTR, bone2Name=pack, swing1LimitDegrees=45, swing2LimitDegrees=45, twistLimitDegrees=45}
 {
  bone1Name=dyn_keychain_MSTR
  bone2Name=pack
  swing1LimitDegrees=45
  swing1Motion=Free
  swing2LimitDegrees=45
  swing2Motion=Free
  twistLimitDegrees=45
  twistMotion=Free
 }
 {bone1Name=dyn_bag_01, bone2Name=pack, swing1LimitDegrees=45, swing2LimitDegrees=45, twistLimitDegrees=45}
 {bone1Name=dyn_bag_02, bone2Name=dyn_bag_01, swing1LimitDegrees=45, swing2LimitDegrees=45, twistLimitDegrees=45}
 {
  bone1Name=dyn_loop_01
  bone2Name=dyn_bag_01
  swing1LimitDegrees=45
  swing1Motion=Free
  swing2LimitDegrees=45
  swing2Motion=Free
  twistLimitDegrees=45
  twistMotion=Free
 }
 {
  bone1Name=dyn_loop_02
  bone2Name=dyn_bag_01
  swing1LimitDegrees=45
  swing1Motion=Free
  swing2LimitDegrees=45
  swing2Motion=Free
  twistLimitDegrees=45
  twistMotion=Free
 }
 {bone1Name=dyn_bagRing, bone2Name=dyn_bagRing_MSTR, swing1LimitDegrees=2, swing2LimitDegrees=80, twistLimitDegrees=2}
 {bone1Name=dyn_rope_top_01, bone2Name=dyn_rope_top_MSTR, swing1LimitDegrees=45, swing2LimitDegrees=45, twistLimitDegrees=45}
 {
  bone1Name=dyn_keychain_01
  bone2Name=dyn_keychain_MSTR
  swing1LimitDegrees=10
  swing1Motion=Free
  swing2LimitDegrees=30
  swing2Motion=Free
  twistLimitDegrees=20
 }
 {
  bone1Name=dyn_rope_BTM_MSTR
  bone2Name=pack
  swing1LimitDegrees=45
  swing1Motion=Free
  swing2LimitDegrees=45
  swing2Motion=Free
  twistLimitDegrees=45
  twistMotion=Free
 }
 {bone1Name=dyn_rope_BTM_01, bone2Name=dyn_rope_BTM_MSTR, swing1LimitDegrees=20, swing2LimitDegrees=20, twistLimitDegrees=2}
 {bone1Name=dyn_rope_BTM_02, bone2Name=dyn_rope_BTM_01, swing1LimitDegrees=20, swing2LimitDegrees=20, twistLimitDegrees=2}
 {bone1Name=dyn_rope_BTM_03, bone2Name=dyn_rope_BTM_02, swing1LimitDegrees=20, swing2LimitDegrees=20, twistLimitDegrees=2}
]
