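# Physics asset for the satchel on the F_MED_Temple backpack cosmetic, providing simplified collision and joint setup for its bones.
# Three capsule bodies cover root, pack and dyn_satchel_Main, with the pack capsule much smaller than the other two.
# Constraints chain them together: pack rotates freely against root, while the satchel's main bone carries tight, equal swing and twist limits against pack.
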PhysicsAsset /BRCosmetics/Accessories/FORT_Backpacks/Backpack_F_MED_Temple/Meshes/F_MED_Temple_Satchel_Physics
bodies {
 dyn_satchel_Main [{shapeType=Capsule, length=14.3924, radius=5.8734, center=<15.3754 0.1625 -0.3374>, rotation=<rot 5.784 -12.4367 -92.7239>}]
 pack [{shapeType=Capsule, length=2, radius=2}]
 root [{shapeType=Capsule, length=24.5058, radius=5.4942, shapeName=root_capsule, center=<0 7.5549 -13.6308>, rotation=<rot 90 -179.9999 179.9999>}]
}
constraints [
 {
  bone1Name=pack
  bone2Name=root
  swing1LimitDegrees=45
  swing1Motion=Free
  swing2LimitDegrees=45
  swing2Motion=Free
  twistLimitDegrees=45
  twistMotion=Free
 }
 {bone1Name=dyn_satchel_Main, bone2Name=pack, swing1LimitDegrees=15, swing2LimitDegrees=15, twistLimitDegrees=15}
]
