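# Animation Blueprint for the F_MED_Temple backpack cosmetic, adding secondary motion to the pack, its eight arrows and satchel.
# After a pack bone offset, low-alpha spring-driven Anim Dynamics nodes jiggle each arrow, followed by RigidBody passes using pack, arrow (partial alpha) and satchel physics assets.
# The update event combines skydiving and dive-mode flags into a bool that enables an extra pack roll, before a slight constant arrow1 roll.
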AnimBlueprint /BRCosmetics/Accessories/FORT_Backpacks/Backpack_F_MED_Temple/Meshes/F_MED_Temple_Pack_AnimBP
target_skeleton=&/F_MED_Temple_Pack_Skeleton
graphs [
 AnimGraph {nodes [
   n1 = "Transform (Modify) Bone - Bone: pack"(ComponentPose=Local To Component(LocalPose=Input Pose()), Translation="0,0.5,2", Rotation="0,0,0") {
    rotationSpace=BCS_BoneSpace
    scaleMode=BMM_Additive
    translationMode=BMM_Additive
    alphaBoolBlend {blendOption=Linear}
    boneToModify {boneName=pack}
    translation {y=0.5, z=2}
   }
   n2 = Anim Dynamics - Bone: arrow1(ComponentPose=@n1, Alpha=0.06, ExternalForce="0,0,0") {
    angularDampingOverride=0.9999
    angularSpringConstant=3500
    bAngularSpring=true
    bGravityOverrideInSimSpace=true
    bOverrideAngularDamping=true
    gravityScale=0.3
    simulationSpace=BoneRelative
    alphaBoolBlend {blendOption=Linear}
    boundBone {boneName=arrow1}
    componentAppliedLinearAccClamp=<2500 2500 2500>
    componentLinearAccScale=<0.5 0.5 0.5>
    componentLinearVelScale=<0.5 0.5 0.5>
    physicsBodyDefinitions [{
      sphereCollisionRadius=0
      boundBone {boneName=arrow1}
      boxExtents=<10 13 13>
      constraintSetup {angularLimitsMax=<5 30 1>, angularLimitsMin=<-5 -30 -1>, angularTarget=<0 0 0>}
      localJointOffset=<0 3 -1>
     }]
    relativeSpaceBone {boneName=root}
   }
   n3 = Anim Dynamics - Bone: arrow2(ComponentPose=@n2, Alpha=0.05, ExternalForce="0,0,0") {
    angularDampingOverride=0.9999
    angularSpringConstant=3500
    bAngularSpring=true
    bGravityOverrideInSimSpace=true
    bOverrideAngularDamping=true
    gravityScale=0.23
    simulationSpace=BoneRelative
    alphaBoolBlend {blendOption=Linear}
    boundBone {boneName=arrow2}
    componentAppliedLinearAccClamp=<2500 2500 2500>
    componentLinearAccScale=<0.5 0.5 0.5>
    componentLinearVelScale=<0.5 0.5 0.5>
    physicsBodyDefinitions [{
      sphereCollisionRadius=0
      boundBone {boneName=arrow2}
      boxExtents=<10 13 13>
      constraintSetup {angularLimitsMax=<10 30 1>, angularLimitsMin=<-10 -30 -1>, angularTarget=<0 0 0>}
      localJointOffset=<-0.1 0 -0.5>
     }]
    relativeSpaceBone {boneName=root}
   }
   n4 = Anim Dynamics - Bone: arrow3(ComponentPose=@n3, Alpha=0.07, ExternalForce="0,0,0") {
    angularDampingOverride=0.9999
    angularSpringConstant=3500
    bAngularSpring=true
    bGravityOverrideInSimSpace=true
    bOverrideAngularDamping=true
    gravityScale=0.28
    simulationSpace=BoneRelative
    alphaBoolBlend {blendOption=Linear}
    boundBone {boneName=arrow3}
    componentAppliedLinearAccClamp=<2500 2500 2500>
    componentLinearAccScale=<0.5 0.5 0.5>
    componentLinearVelScale=<0.5 0.5 0.5>
    physicsBodyDefinitions [{
      sphereCollisionRadius=0
      boundBone {boneName=arrow3}
      boxExtents=<10 13 13>
      constraintSetup {angularLimitsMax=<10 30 1>, angularLimitsMin=<-10 -30 -1>, angularTarget=<0 0 0>}
      localJointOffset=<-0.2 0 -0.9>
     }]
    relativeSpaceBone {boneName=root}
   }
   n5 = Anim Dynamics - Bone: arrow4(ComponentPose=@n4, Alpha=0.05, ExternalForce="0,0,0") {
    angularDampingOverride=0.9999
    angularSpringConstant=3500
    bAngularSpring=true
    bGravityOverrideInSimSpace=true
    bOverrideAngularDamping=true
    gravityScale=0.25
    simulationSpace=BoneRelative
    alphaBoolBlend {blendOption=Linear}
    boundBone {boneName=arrow4}
    componentAppliedLinearAccClamp=<2500 2500 2500>
    componentLinearAccScale=<0.5 0.5 0.5>
    componentLinearVelScale=<0.5 0.5 0.5>
    physicsBodyDefinitions [{
      sphereCollisionRadius=0
      boundBone {boneName=arrow4}
      boxExtents=<10 13 13>
      constraintSetup {angularLimitsMax=<10 30 1>, angularLimitsMin=<-10 -30 -1>, angularTarget=<0 0 0>}
      localJointOffset=<-0.1 0 -0.6>
     }]
    relativeSpaceBone {boneName=root}
   }
   n6 = Anim Dynamics - Bone: arrow5(ComponentPose=@n5, Alpha=0.085, ExternalForce="0,0,0") {
    angularDampingOverride=0.9999
    angularSpringConstant=3500
    bAngularSpring=true
    bGravityOverrideInSimSpace=true
    bOverrideAngularDamping=true
    gravityScale=0.28
    simulationSpace=BoneRelative
    alphaBoolBlend {blendOption=Linear}
    boundBone {boneName=arrow5}
    componentAppliedLinearAccClamp=<2500 2500 2500>
    componentLinearAccScale=<0.5 0.5 0.5>
    componentLinearVelScale=<0.5 0.5 0.5>
    physicsBodyDefinitions [{
      sphereCollisionRadius=0
      boundBone {boneName=arrow5}
      boxExtents=<10 13 13>
      constraintSetup {angularLimitsMax=<10 30 1>, angularLimitsMin=<-10 -30 -1>, angularTarget=<0 0 0>}
      localJointOffset=<-0.1 0 -1>
     }]
    relativeSpaceBone {boneName=root}
   }
   n7 = Anim Dynamics - Bone: arrow6(ComponentPose=@n6, Alpha=0.085, ExternalForce="0,0,0") {
    angularDampingOverride=0.9999
    angularSpringConstant=3500
    bAngularSpring=true
    bGravityOverrideInSimSpace=true
    bOverrideAngularDamping=true
    gravityScale=0.25
    simulationSpace=BoneRelative
    alphaBoolBlend {blendOption=Linear}
    boundBone {boneName=arrow6}
    componentAppliedLinearAccClamp=<2500 2500 2500>
    componentLinearAccScale=<0.5 0.5 0.5>
    componentLinearVelScale=<0.5 0.5 0.5>
    physicsBodyDefinitions [{
      sphereCollisionRadius=0
      boundBone {boneName=arrow6}
      boxExtents=<10 13 13>
      constraintSetup {angularLimitsMax=<5 30 1>, angularLimitsMin=<-5 -30 -1>, angularTarget=<0 0 0>}
      localJointOffset=<-0.1 0 -0.8>
     }]
    relativeSpaceBone {boneName=root}
   }
   n8 = Anim Dynamics - Bone: arrow7(ComponentPose=@n7, Alpha=0.085, ExternalForce="0,0,0") {
    angularDampingOverride=0.9999
    angularSpringConstant=3500
    bAngularSpring=true
    bGravityOverrideInSimSpace=true
    bOverrideAngularDamping=true
    gravityScale=0.28
    simulationSpace=BoneRelative
    alphaBoolBlend {blendOption=Linear}
    boundBone {boneName=arrow7}
    componentAppliedLinearAccClamp=<2500 2500 2500>
    componentLinearAccScale=<0.5 0.5 0.5>
    componentLinearVelScale=<0.5 0.5 0.5>
    physicsBodyDefinitions [{
      sphereCollisionRadius=0
      boundBone {boneName=arrow7}
      boxExtents=<10 13 13>
      constraintSetup {angularLimitsMax=<10 30 1>, angularLimitsMin=<-10 -30 -1>, angularTarget=<0 0 0>}
      localJointOffset=<-0.2 0 -1>
     }]
    relativeSpaceBone {boneName=root}
   }
   n9 = Anim Dynamics - Bone: arrow8(ComponentPose=@n8, Alpha=0.07, ExternalForce="0,0,0") {
    angularDampingOverride=0.9999
    angularSpringConstant=3500
    bAngularSpring=true
    bGravityOverrideInSimSpace=true
    bOverrideAngularDamping=true
    gravityScale=0.23
    simulationSpace=BoneRelative
    alphaBoolBlend {blendOption=Linear}
    boundBone {boneName=arrow8}
    componentAppliedLinearAccClamp=<2500 2500 2500>
    componentLinearAccScale=<0.5 0.5 0.5>
    componentLinearVelScale=<0.5 0.5 0.5>
    physicsBodyDefinitions [{
      sphereCollisionRadius=0
      boundBone {boneName=arrow8}
      boxExtents=<10 13 13>
      constraintSetup {angularLimitsMax=<10 30 1>, angularLimitsMin=<-10 -30 -1>, angularTarget=<0 0 0>}
      localJointOffset=<-0.1 0 -0.85>
     }]
    relativeSpaceBone {boneName=root}
   }
   n10 = RigidBody(ComponentPose=@n9) {
    bClampLinearTranslationLimitToRefPose=true
    bOverrideWorldGravity=true
    overridePhysicsAsset=&/F_MED_Temple_Pack_Physics
    simulationSpace=BaseBoneSpace
    alphaBoolBlend {blendOption=Linear}
    baseBoneRef {boneName=root}
    componentAppliedLinearAccClamp=<2500 2500 2500>
    componentLinearAccScale=<0.4 0.4 0.5>
    overrideWorldGravity {z=-980}
   }
   n11 = RigidBody(ComponentPose=@n10, Alpha=0.2) {
    bClampLinearTranslationLimitToRefPose=true
    bOverrideWorldGravity=true
    overridePhysicsAsset=&/F_MED_Temple_Arrow_Physics
    simulationSpace=BaseBoneSpace
    alphaBoolBlend {blendOption=Linear}
    baseBoneRef {boneName=root}
    componentAppliedLinearAccClamp=<2500 2500 2500>
    componentLinearAccScale=<0.3 0.3 0.4>
    overrideWorldGravity {z=-980}
   }
   n12 = RigidBody(ComponentPose=@n11) {
    bClampLinearTranslationLimitToRefPose=true
    bOverrideWorldGravity=true
    overridePhysicsAsset=&/F_MED_Temple_Satchel_Physics
    simulationSpace=BaseBoneSpace
    alphaBoolBlend {blendOption=Linear}
    baseBoneRef {boneName=root}
    componentAppliedLinearAccClamp=<2500 2500 2500>
    componentLinearAccScale=<0.3 0.3 0.4>
    overrideWorldGravity {z=-980}
   }
   n13 = "Transform (Modify) Bone - Bone: pack"!(ComponentPose=@n12, bAlphaBoolEnabled:bool=Get Is Skydivee() -> (Is Skydivee:bool), Translation="0,0,0", Rotation="0,0,-14", Scale) -> (Pose) {
    alphaInputType=Bool
    rotationMode=BMM_Additive
    scaleMode=BMM_Additive
    translationMode=BMM_Additive
    alphaBoolBlend {blendOption=Linear}
    boneToModify {boneName=pack}
    rotation {roll=-14}
   }
   n14 = "Transform (Modify) Bone - Bone: arrow1"(ComponentPose=@n13, Alpha=1, Translation="0,0,0", Rotation="0,0,1", Scale="1,1,1") {
    rotationMode=BMM_Additive
    scaleMode=BMM_Additive
    translationMode=BMM_Additive
    alphaBoolBlend {blendOption=Linear}
    boneToModify {boneName=arrow1}
    rotation {roll=1}
   }
   Output Pose(Result=Component To Local(ComponentPose=@n14)) {layerGroup=None}
  ]}
 EventGraph {nodes [
   Event BlueprintUpdateAnimation()
   > Sequence()
   > Set Is Skydivee(Is Skydivee:bool=OR Boolean(A=Get bIsSkyDiving(), B=Get bIsSkydiveDiveMode())) -> (Output_Get:bool)
  ]}
]
variables [Is Skydivee {type=boolean}]
